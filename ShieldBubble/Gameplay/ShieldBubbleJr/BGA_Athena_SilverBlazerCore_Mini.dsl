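; Core building gameplay actor for the mini Shield Bubble (Shield Bubble Jr), which supplies the bubble safe zone to spawn.
; SpawnBubbleSafeZone spawns BGA_Athena_SilverBlazer_Mini at this actor's location with default rotation and scale, passes along its associated mission and instigator, and returns the new actor.
; The event graph holds only empty stubs and the construction script just calls its parent, so this Blueprint adds no other logic.
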(event EventBeginPlay
  (|CallParentFunction) ; This node is disabled and will not be called.\nDrag off pins to build functionality.
)

(event Collision|EventActorBeginOverlap
  (params (OtherActor :type Actor%20Object%20Reference)))

(event EventTick
  (params (DeltaSeconds :type Float%20%28single-precision%29)))

(fn SpawnBubbleSafeZone
  (bind _returnvalue_location (Math|Transform|BreakTransform (Transformation|GetActorTransform)))
  (bind _returnvalue (Game|SpawnBuildingActorfromClass "/ShieldBubble/Gameplay/ShieldBubbleJr/BGA_Athena_SilverBlazer_Mini.BGA_Athena_SilverBlazer_Mini_C" (Math|Transform|MakeTransform _returnvalue_location) 0 false 0.0 (Variables|Missions|GetAssociatedMissionParam) (Game|GetInstigator)))
  (return _returnvalue))

(fn ConstructionScript
  (|CallParentFunction))
