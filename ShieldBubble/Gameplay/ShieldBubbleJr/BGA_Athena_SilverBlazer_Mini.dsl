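; Mini variant of the shield bubble actor (SilverBlazer, in the ShieldBubbleJr folder), built as a child of an existing parent Blueprint.
; After the parent construction script, it uniformly scales the root by its MaxScale over MaxScaleParent, both read from scalable floats via a safe divide.
; It also sets OverallOpacity on the SafeZone materials to a fixed partial value.
; The BeginPlay, overlap and tick events add no logic of their own.
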(event EventBeginPlay
  (|CallParentFunction) ; This node is disabled and will not be called.\nDrag off pins to build functionality.
)

(event Collision|EventActorBeginOverlap
  (params (OtherActor :type Actor%20Object%20Reference)))

(event EventTick
  (params (DeltaSeconds :type Float%20%28single-precision%29)))

(fn ConstructionScript
  (|CallParentFunction)
  (Transformation|SetWorldScale3D (Variables|Default|GetRoot) (Math|Conversions|ToVector(Float) (Math|Float|SafeDivide (ScalableFloat|GetValueatLevel (Variables|SilverBronco|GetMaxScale)) (ScalableFloat|GetValueatLevel (Variables|Default|GetMaxScaleParent)))))
  (Rendering|Material|SetScalarParameterValueonMaterials (Variables|Default|GetSafeZone) "OverallOpacity" 0.6))
